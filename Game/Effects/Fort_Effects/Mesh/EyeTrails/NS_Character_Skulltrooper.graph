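# Niagara system for glowing eye and eye-trail effects on the Skulltrooper character.
# It holds three emitters, EyeL, EyeR and EyeTrail, and exposes an HDR EyeColor (cyan-blue by default) and a bisHeadHidden flag as user parameters.
# The system uses SystemState, a set-variables assignment in its spawn script and a bool-to-int dynamic input, likely converting the head-hidden flag.
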NiagaraSystem /Game/Effects/Fort_Effects/Mesh/EyeTrails/NS_Character_Skulltrooper
dependencies {
 usedDataInterfaces []
 usedDynamicInputs [&MakeCustomIntFromBool]
 usedModules [
  &SystemState
  &self:SystemSpawnScript.SystemScriptSource.SystemScriptGraph.NiagaraNodeAssignment_0.SetVariables_#1
 ]
}
emitters {EyeL {rendererClasses [], renderers []}, EyeR {rendererClasses [], renderers []}, EyeTrail {rendererClasses [], renderers []}}
summary {
 systemName=NS_Character_Skulltrooper
 emitters [{emitterName=EyeL, rendererClasses []}, {emitterName=EyeR, rendererClasses []}, {emitterName=EyeTrail, rendererClasses []}]
 userVariables [
  User.bisHeadHidden {type=NiagaraBool}
  User.EyeColor {type=LinearColor, description="", defaultValue {struct=LinearColor, value=<rgba 0 3.0924 5 1>}}
 ]
}
